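# One-shot stereo MetaSound Source for a Winterfest UI whoosh in the Season 28 Battle Pass quests content.
# On play, a Wave Player plays a generic item shop whoosh pitched up two octaves, feeds both outputs and signals finish when done.
# It uses the UI sound class, though its volume is modulated by the music volume patch.
# A concurrency override keeps a single instance, stopping the oldest, with a short retrigger time.
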MetaSoundSource /BattlePassS28_Quests/Audio/MSS_S28_UI_Winterfest_Whoosh
inputs [
 UE.Source.OnPlay {type=Trigger}
 Wave Asset {type=WaveAsset}
]
interfaces [UE.OutputFormat.Stereo:1.0, UE.Source.OneShot:1.0, UE.Source:1.1]
outputs [
 "UE.OutputFormat.Stereo.Audio:0" {type=Audio}
 "UE.OutputFormat.Stereo.Audio:1" {type=Audio}
 UE.Source.OneShot.OnFinished {type=Trigger}
]
pages [Default {
  input_defaults {UE.Source.OnPlay=null, Wave Asset=/Game/Sounds/Fort_UI_Sounds/FrontEnd/ItemShop/wav/Fort_UI_Shop_Whoosh_Generic_01}
  nodes [WavePlayer_0 = UE.Wave Player:Stereo(Play=^UE.Source.OnPlay, Wave Asset=^Wave Asset) {also {defaults {Pitch Shift=24}}}]
  links [
   {dst="@UE.Source.OneShot.OnFinished", src=WavePlayer_0.On Finished}
   {dst="@UE.OutputFormat.Stereo.Audio:0", src=WavePlayer_0.Out Left}
   {dst="@UE.OutputFormat.Stereo.Audio:1", src=WavePlayer_0.Out Right}
  ]
 }]
properties {
 assetImportData=&self:AssetImportData
 bOverrideConcurrency=true
 numChannels=2
 outputFormat=Stereo
 sampleRate=48000
 soundClassObject=&Fort_UI
 volume=0.7
 concurrencyOverrides {maxCount=1, resolutionRule=StopOldest, retriggerTime=0.11}
 modulationSettings {volumeModulationDestination {bEnableModulation=true, modulators [/Game/Sounds/Modulation/Patches/Patch_FortMusic_Vol]}}
}
